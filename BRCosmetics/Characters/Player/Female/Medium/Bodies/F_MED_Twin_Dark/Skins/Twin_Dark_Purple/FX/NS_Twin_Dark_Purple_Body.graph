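# Niagara body effect for the Twin Dark Purple character skin, emitting short-lived pale pink sprite sparkles around the legs and pelvis.
# Sparkles_Legs places particles with FN_BoneFlow on the user skeletal mesh, using a curve-biased random height to drive radius and camera offset, with spawn-group-masked mirrored offsets.
# Sparkles_Pelvis re-samples one bone each update, adding a scaled sphere offset chosen at spawn.
# Both animate sprite size over life and enlarge it with LOD distance.
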NiagaraSystem /BRCosmetics/Characters/Player/Female/Medium/Bodies/F_MED_Twin_Dark/Skins/Twin_Dark_Purple/FX/NS_Twin_Dark_Purple_Body
dependencies {
 usedDataInterfaces [&NiagaraDataInterfaceCurve, &NiagaraDataInterfaceVector2DCurve]
 usedDynamicInputs []
 usedModules [&SystemState]
}
emitters {
 Sparkles_Legs {
  emitterSpawnScript [&FN_PreserveSkydiveUserVariables() {module=PreserveUserVariables}]
  emitterUpdateScript [
   &EmitterState()
   &SpawnRate(SpawnRate=4, Use Spawn Probability=false)
  ]
  particleSpawnScript [
   &FN_SetSpawnGroup(Number of Spawn Groups=2)
   &InitializeParticle(
    Color=<rgba 1 0.6042 0.9784 1>
    Lifetime Max=1.2
    Lifetime Min=0.6
    Lifetime Mode=::NewEnumerator1
    Sprite Size Mode=::NewEnumerator1
    Uniform Sprite Size Max=14
    Uniform Sprite Size Min=8
    UsePositionOffset=false
    Write Material Random=true
   )
   /SetVariables(Particles.BoneFlowFloat:NiagaraFloat=&FloatFromCurve(
     FloatCurve=DataInterface() {propertyValues {Curve {keys [
         {interpMode=RCIM_Cubic, tangentMode=RCTM_User, time=0, value=0}
         {arriveTangent=2.3655, interpMode=RCIM_Cubic, leaveTangent=2.3655, tangentMode=RCTM_User, time=1, value=1}
        ]}}}
     CurveIndex=&RandomRangeFloat(Minimum=0, Maximum=1)
    )) {isSetParameters=true}
   &CameraOffset(Camera Offset Amount=&Lerp_Float(
     A=0
     B=8
     Alpha=&FloatFromCurve(
      FloatCurve=DataInterface() {propertyValues {Curve {keys [
          {arriveTangent=4.1785, interpMode=RCIM_Cubic, leaveTangent=4.1785, tangentMode=RCTM_User, time=0, value=0}
          {arriveTangent=1, leaveTangent=1, time=0.1, value=0.9}
          {arriveTangent=-0.0405, interpMode=RCIM_Cubic, leaveTangent=-0.0405, tangentMode=RCTM_User, time=1, value=1}
         ]}}}
      CurveIndex=$Particles.BoneFlowFloat
     )
    ))
  ]
  particleUpdateScript [
   &ParticleState()
   &FN_BoneFlow(
    Angle:NiagaraFloat=&RandomRangeFloat(Minimum=0, Maximum=1)
    Enable Position Offsets=true
    Height:NiagaraFloat=$Particles.BoneFlowFloat
    Interpolate Mesh Sampling=false
    Offset End:Vector3f=&MaskVectorBySpawnGroup(Masked Vector=<-22 -8 0>, Passthrough Vector=<22 8 0>)
    Offset Start:Vector3f=&MaskVectorBySpawnGroup(Masked Vector=<-2.5 0 0>, Passthrough Vector=<2.5 0 0>)
    Radius:NiagaraFloat=&FloatFromCurve(
     FloatCurve=DataInterface() {propertyValues {Curve {keys [
         {arriveTangent=4.6102, interpMode=RCIM_Cubic, leaveTangent=4.6102, tangentMode=RCTM_User, time=-0.01, value=1.9}
         {
          arriveTangent=1.2997
          arriveTangentWeight=0.1952
          interpMode=RCIM_Cubic
          leaveTangent=1.2997
          leaveTangentWeight=0.1952
          tangentMode=RCTM_User
          tangentWeightMode=RCTWM_WeightedBoth
          time=0.5
          value=1.3
         }
         {arriveTangent=-5.5891, interpMode=RCIM_Cubic, leaveTangent=-5.5891, tangentMode=RCTM_User, time=1, value=1}
        ]}}}
     CurveIndex=$Particles.BoneFlowFloat
     Scale Curve=3
    )
    Set Material Dynamic=false
    SkeletalMesh:NiagaraDataInterfaceSkeletalMesh=$User.SkeletalMesh
   )
   &ScaleSpriteSize(Scale Factor=&Vector2DFromCurve(
     Vector2Curve=DataInterface() {propertyValues {
       XCurve {keys [
         {arriveTangent=25.6549, interpMode=RCIM_Cubic, leaveTangent=25.6549, tangentMode=RCTM_User, time=0, value=0.75}
         {arriveTangent=1, interpMode=RCIM_Cubic, leaveTangent=1, time=0.1, value=0.28}
         {arriveTangent=1.3585, leaveTangent=1.3585, time=0.25, value=1}
         {arriveTangent=-1.3333, interpMode=RCIM_Cubic, leaveTangent=-1.3333, time=0.63, value=1}
         {arriveTangent=-5.3669, interpMode=RCIM_Cubic, leaveTangent=-5.3669, tangentMode=RCTM_User, time=1, value=0}
        ]}
       YCurve {keys [
         <at 0 2 RCIM_Cubic>
         {arriveTangent=-4, interpMode=RCIM_Cubic, leaveTangent=-4, time=0.1, value=1}
         <at 0.25 1>
         {arriveTangent=-1.3333, interpMode=RCIM_Cubic, leaveTangent=-1.3333, time=0.63, value=1}
         {arriveTangent=-5.3669, interpMode=RCIM_Cubic, leaveTangent=-5.3669, tangentMode=RCTM_User, time=1, value=0}
        ]}
      }}
     ScaleCurve=<1 1>
    ))
   &ScaleSpriteSize(
    Scale Sprite Size Mode=::NewEnumerator0
    Uniform Scale Factor=&FloatFromCurve(
     FloatCurve=DataInterface() {propertyValues {Curve {keys [<at 500 1 RCIM_Cubic>, {interpMode=RCIM_Cubic, leaveTangent=0.0207, time=1400, value=3}]}}}
     CurveIndex=$Engine.Owner.LODDistance
    )
   ) {module=ScaleSpriteSize001}
  ]
  rendererClasses [&NiagaraSpriteRendererProperties]
  renderers [{rendererClass=&NiagaraSpriteRendererProperties}]
 }
 Sparkles_Pelvis {
  emitterSpawnScript [&FN_PreserveSkydiveUserVariables() {module=PreserveUserVariables}]
  emitterUpdateScript [
   &EmitterState()
   &SpawnRate(Spawn Probability=0.8, SpawnRate=2, Use Spawn Probability=true)
  ]
  particleSpawnScript [
   &FN_SetSpawnGroup(Number of Spawn Groups=2)
   &InitializeParticle(
    Color=<rgba 1 0.6042 0.9784 1>
    Lifetime Max=1.2
    Lifetime Min=0.8
    Lifetime Mode=::NewEnumerator1
    Sprite Size Mode=::NewEnumerator1
    Uniform Sprite Size Max=14
    Uniform Sprite Size Min=8
    UsePositionOffset=false
    Write Material Random=true
   )
   &CameraOffset(Camera Offset Amount=8)
   &ShapeLocation(
    Hemisphere Distribution=<0.5 1>
    MaskSpawnGroup=false
    Non Uniform Scale=<0.5 0.35 1>
    Offset=<-4 -5 0>
    Offset Mode=::NewEnumerator0
    Override Seed=false
    Rotation Mode=::NewEnumerator2
    Sphere Radius=14
    Sphere Surface Distribution=1
    UseOwnerScale=true
    Yaw / Pitch / Roll=<0.5 0 0>
   )
  ]
  particleUpdateScript [
   &ParticleState()
   &SkeletalMeshLocation(
    Bone / Socket Index=6
    Bone Sampling Mode=::NewEnumerator4
    MaskSpawnGroup=false
    Sampled Position Offset=$Particles.ShapeLocation.ShapeVector
    Skeletal Mesh=$User.SkeletalMesh
   )
   &ScaleSpriteSize(Scale Factor=&Vector2DFromCurve(
     Vector2Curve=DataInterface() {propertyValues {
       XCurve {keys [
         {arriveTangent=25.6549, interpMode=RCIM_Cubic, leaveTangent=25.6549, tangentMode=RCTM_User, time=0, value=0.75}
         {arriveTangent=1, interpMode=RCIM_Cubic, leaveTangent=1, time=0.1, value=0.28}
         {arriveTangent=1.3585, leaveTangent=1.3585, time=0.25, value=1}
         {arriveTangent=-1.3333, interpMode=RCIM_Cubic, leaveTangent=-1.3333, time=0.63, value=1}
         {arriveTangent=-5.3669, interpMode=RCIM_Cubic, leaveTangent=-5.3669, tangentMode=RCTM_User, time=1, value=0}
        ]}
       YCurve {keys [
         <at 0 2 RCIM_Cubic>
         {arriveTangent=-4, interpMode=RCIM_Cubic, leaveTangent=-4, time=0.1, value=1}
         <at 0.25 1>
         {arriveTangent=-1.3333, interpMode=RCIM_Cubic, leaveTangent=-1.3333, time=0.63, value=1}
         {arriveTangent=-5.3669, interpMode=RCIM_Cubic, leaveTangent=-5.3669, tangentMode=RCTM_User, time=1, value=0}
        ]}
      }}
     ScaleCurve=<1 1>
    ))
   &ScaleSpriteSize(
    Scale Sprite Size Mode=::NewEnumerator0
    Uniform Scale Factor=&FloatFromCurve(
     FloatCurve=DataInterface() {propertyValues {Curve {keys [<at 500 1 RCIM_Cubic>, {interpMode=RCIM_Cubic, leaveTangent=0.0207, time=1400, value=3}]}}}
     CurveIndex=$Engine.Owner.LODDistance
    )
   ) {module=ScaleSpriteSize001}
  ]
  rendererClasses [&NiagaraSpriteRendererProperties]
  renderers [{rendererClass=&NiagaraSpriteRendererProperties}]
 }
}
summary {
 systemName=NS_Twin_Dark_Purple_Body
 emitters [
  {emitterName=Sparkles_Legs, rendererClasses [&NiagaraSpriteRendererProperties]}
  {emitterName=Sparkles_Pelvis, rendererClasses [&NiagaraSpriteRendererProperties]}
 ]
 userVariables [
  User.bIsFrontEndPreview {type=NiagaraBool}
  User.PlayerSpeed {type=NiagaraFloat, defaultValue=0, description=""}
  User.SkeletalMesh {
   type=NiagaraDataInterfaceSkeletalMesh
   defaultValue {
    struct=NiagaraExt_VariableValue_DataInterface
    value {dataInterface=&self:NiagaraDataInterfaceSkeletalMesh_0}
   }
  }
 ]
}
